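; Figment water actor Blueprint that sets up its underwater-style post-process material to match the water plane's height at BeginPlay.
; It creates a dynamic instance from the post-process component's first blendable material, falling back to the default Athena water post-process material.
; It sets WaterLevel from the plane's world Z minus an offset, then writes the instance back as the sole full-weight blendable.
; BeginPlay also zeroes BounceLake, and construction only calls the parent.
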(event EventBeginPlay
  (bind _waterpost (Variables|Default|GetWaterPost))
  (bind _settings (Variables|/Script/Engine.PostProcessComponent|GetSettings _waterpost))
  (bind _weightedblendables (Utilities|Struct|BreakPostProcessSettings _settings))
  (bind _array (Utilities|Struct|BreakWeightedBlendables _weightedblendables))
  (bind (_weight _object) (Utilities|Struct|BreakWeightedBlendable (Utilities|Array|Get(aref) _array)))
  ;[1 1 1] Begin Play set PostProcess Water Level
  (Variables|Default|SetBounceLake 0)
  (if (> (Utilities|Array|Length _array) 0)
    (bind _asmaterial_interface (Utilities|Casting|CastToMaterialInterface _object)
      (:then
        _)
      (:CastFailed
        _))
    (bind _returnvalue (Rendering|Material|CreateDynamicMaterialInstance (select (Utilities|IsValid _asmaterial_interface :!class /Script/Engine.KismetSystemLibrary) _asmaterial_interface "/Game/Athena/Environments/Materials/Water/M_Athena_Water_PostProcess_0.M_Athena_Water_PostProcess_0") "None" "None" :!class /Script/Engine.KismetMaterialLibrary))
    (bind _output_get (Variables|Default|SetPP%20MID _returnvalue))
    (else
      (bind _returnvalue_2 (Rendering|Material|CreateDynamicMaterialInstance "/Game/Athena/Environments/Materials/Water/M_Athena_Water_PostProcess_0.M_Athena_Water_PostProcess_0" "None" "None" :!class /Script/Engine.KismetMaterialLibrary))
      (Variables|Default|SetPP%20MID _returnvalue_2)))
  (Rendering|Material|SetScalarParameterValue _output_get "WaterLevel" (- (.z (Transformation|GetWorldLocation (Variables|Default|GetWater_Plane))) (Variables|Default|GetPost%20Process%20Offset)) :!class /Script/Engine.MaterialInstanceDynamic)
  (bind _structout (Utilities|Struct|SetmembersinPostProcessSettings (Variables|/Script/Engine.PostProcessComponent|GetSettings (Variables|Default|GetWaterPost)) (Utilities|Struct|MakeWeightedBlendables (Utilities|Array|MakeArray (Utilities|Struct|MakeWeightedBlendable 1.0 (Variables|Default|GetPP%20MID))))))
  (Variables|/Script/Engine.PostProcessComponent|SetSettings _structout (Variables|Default|GetWaterPost)))

(fn ConstructionScript
  (|CallParentFunction))
